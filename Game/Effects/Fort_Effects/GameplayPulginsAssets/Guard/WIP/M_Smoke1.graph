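# Translucent unlit smoke material for Niagara sprites and ribbons, in a WIP Guard effects folder.
# A dynamic Frame value drives a 4x5 flipbook, blending adjacent panning frames using explicit UV derivatives to avoid seams.
# Param2 blends evil-smoke texture into that flipbook for dark tinted smoke, while a second flipbook adds refined, tinted glow scaled by Emis.
# Particle color modulates the result, with depth fade and inverse Fresnel softening opacity.
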Material /Game/Effects/Fort_Effects/GameplayPulginsAssets/Guard/WIP/M_Smoke1
graph {
 properties {
  bAutomaticallySetUsageInEditor=false
  bUsedWithNiagaraRibbons=true
  bUsedWithNiagaraSprites=true
  blendMode=BLEND_Translucent
  shadingModel=MSM_Unlit
  usedShadingModels=Unlit
 }
 nodes [
  n1 = VectorParameter() {group=R, parameterName=R_UVControl, channelNames {a=VPan, b=UPan, g=VScale, r=UScale}, defaultValue=<rgba 1 1 0 0>}
  n2 = MaterialFunctionCall(
   Texture=TextureObject() {samplerType=SAMPLERTYPE_LinearColor, texture=/Game/Effects/Fort_Effects/GameplayPulginsAssets/Guard/Ribbon/T_EvilSmoke}
   UVControl_RGB=@n1
   UVControl_A=@n1
  ) {materialFunction=&FXCommon_TextureSample}
  n3 = DynamicParameter() {defaultValue=<rgba 0 1 1 1>, paramNames [Frame,Param2,Emis,Param4]}
  n4 = Multiply(A=Frac(None=@n3.Frame)) {constB=20}
  n5 = Frac(None=@n4) {desc=Frame Lerp}
  n6 = Subtract(A=@n4, B=@n5) {desc=IndexA}
  n7 = Constant2Vector() {g=5, r=4}
  n8 = ComponentMask(None=@n7) {desc="## Horz Images", r=true}
  n9 = Divide(A=Constant2Vector() {g=1, r=1}, B=@n7) {desc=Scale}
  n10 = ComponentMask(None=@n9) {desc="1 / ## Horz Images", r=true}
  n11 = Multiply(A=@n6, B=@n10)
  n12 = Frac(None=TextureCoordinate())
  n13 = Add(A=AppendVector(A=Fmod(A=@n6, B=@n8) {desc=IndexA_Horz}, B=Floor(None=@n11) {desc=IndexA_Vert}) {desc=IndexA_Combined}, B=@n12)
  n14 = AppendVector(A=@n1.UPan, B=@n1.UPan)
  n15 = Multiply(A=Divide(A=@n14, B=@n7), B=Time() {bOverride_Period=true, period=6666})
  n16 = Frac(None=Add(A=Multiply(A=@n13, B=@n9) {desc=UV_A}, B=@n15))
  n17 = TextureObject() {samplerType=SAMPLERTYPE_LinearColor, texture=&/Frame_13}
  n18 = TextureCoordinate()
  n19 = DDX(Value=@n18)
  n20 = DDY(Value=@n18)
  n21 = TextureSample!(UVs=@n16, Tex=@n17, "DDX(UVs)"=@n19, "DDY(UVs)"=@n20, Apply View MipBias) {mipValueMode=TMVM_Derivative, texture=/Engine/EngineMaterials/DefaultDiffuse}
  n22 = Add(B=@n6) {constA=1, desc=IndexB}
  n23 = Multiply(A=@n22, B=@n10)
  n24 = Add(A=@n12, B=AppendVector(A=Fmod(A=@n22, B=@n8) {desc=IndexB_Horz}, B=Floor(None=@n23) {desc=IndexB_Vert}) {desc=IndexB_Combined})
  n25 = Frac(None=Add(A=@n15, B=Multiply(A=@n24, B=@n9) {desc=UV_B}))
  n26 = TextureSample(UVs=@n25, Tex=@n17, "DDX(UVs)"=@n19, "DDY(UVs)"=@n20) {mipValueMode=TMVM_Derivative, texture=&DefaultDiffuse}
  n27 = LinearInterpolate(A=@n2.A, B=LinearInterpolate(A=@n21.A, B=@n26.A, Alpha=@n5), Alpha=@n3.Param2)
  n28 = ParticleColor()
  n29 = Fresnel() {baseReflectFraction=0}
  Multiply(
   A=DepthFade(Opacity=Multiply(A=@n27, B=@n28.A), FadeDistance=ScalarParameter() {defaultValue=100, parameterName=DF})
   B=Power(Base=OneMinus(None=@n29))
  )
  Multiply(A=@n14)
  n30 = TextureObject() {samplerType=SAMPLERTYPE_LinearColor, texture=/Game/Effects/Fort_Effects/GameplayPulginsAssets/Guard/Ribbon/Trail_Particle_FC}
  n31 = TextureSample(UVs=@n16, Tex=@n30, "DDX(UVs)"=@n19, "DDY(UVs)"=@n20) {mipValueMode=TMVM_Derivative, texture=&DefaultDiffuse}
  n32 = TextureSample(UVs=@n25, Tex=@n30, "DDX(UVs)"=@n19, "DDY(UVs)"=@n20) {mipValueMode=TMVM_Derivative, texture=&DefaultDiffuse}
  n33 = Multiply(A=LinearInterpolate(A=@n31.RGB, B=@n32.RGB, Alpha=@n5), B=LinearInterpolate(A=@n31.A, B=@n32.A, Alpha=@n5))
  n34 = VectorParameter() {parameterName=FC, defaultValue=<rgba 5 50 0.1 0.5>}
  n35 = Multiply(
   A=DynamicParameter() {defaultValue=<rgba 0 1 1 1>, paramNames [Frame,Param2,Emis,Param4]}.Emis
   B=Multiply(
    A=VectorParameter() {parameterName=g354, defaultValue=<rgba 0.5 1.55 5 1>}.RGB
    B=MaterialFunctionCall(Texture=@n33, RGB=@n34.RGB, A=@n34.RGB) {materialFunction=&MF_FX_Refine}.Result
   )
  )
  n36 = LinearInterpolate(A=@n21.RGB, B=@n26.RGB, Alpha=@n5)
  n37 = Multiply(
   A=VectorParameter() {parameterName=d, defaultValue=<rgba 0.04 0.041 0.05 1>}.RGB
   B=Power(Base=LinearInterpolate(A=@n2.RGB, B=@n36, Alpha=@n3.Param2)) {constExponent=0.2}
  )
  Multiply(
   A=MaterialFunctionCall(Emissive=Add(A=@n35, B=@n37)) {materialFunction=&MF_FX_EmissiveAdjustments}.Result
   B=@n28.RGB
  )
  ParticleRelativeTime()
  ScalarParameter() {parameterName=f23}
  Time()
 ]
}
